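# Anim Blueprint for the ToadCycle ArmC female face accessory, adding secondary motion to front hair on top of the attached parent mesh's pose.
# An additive scarf bone offset feeds eight spring-driven AnimDynamics front-hair nodes, then a head-space RigidBody node using the accessory's physics asset.
# Each update sets arm and scarf toggles from cosmetic meta tags, though the shown AnimGraph never reads them.
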AnimBlueprint /BRCosmetics/Characters/Player/Female/Medium/Bodies/F_MED_ToadCycle_ArmC/Meshes/Parts/F_MED_ToadCycle_ArmC_FaceAcc_AnimBP
target_skeleton=&/F_MED_ToadCycle_ArmC_FaceAcc_Skeleton
graphs [
 AnimGraph {nodes [
   n1 = "Transform (Modify) Bone - Bone: scarf"(
    ComponentPose=Local To Component(LocalPose=Copy Pose From Mesh() {bUseAttachedParent=true})
    Alpha=1
    Translation="1.5,1,0"
    Rotation="-20,0,0"
    Scale="1,1,1"
   ) {
    rotationMode=BMM_Additive
    rotationSpace=BCS_BoneSpace
    scaleMode=BMM_Additive
    scaleSpace=BCS_BoneSpace
    translationMode=BMM_Additive
    translationSpace=BCS_BoneSpace
    alphaBoolBlend {blendInTime=0.2, blendOutTime=0.2}
    boneToModify {boneName=scarf}
    rotation {pitch=-20}
    translation=<1.5 1>
   }
   n2 = Anim Dynamics - Bone: dyn_fronthairA_01!(ComponentPose=@n1, Alpha=1, GravityOverride:Vector="0,0,-980", ExternalForce="0,0,0") -> (Pose) {
    angularDampingOverride=0.999
    angularSpringConstant=800
    bAngularSpring=true
    bOverrideAngularDamping=true
    bUseGravityOverride=true
    gravityScale=0.3
    simulationSpace=BoneRelative
    alphaBoolBlend {blendOption=Linear}
    boundBone {boneName=dyn_fronthairA_01}
    componentAppliedLinearAccClamp=<2500 2500 1500>
    componentLinearAccScale=<0.5 0.5 0.4>
    componentLinearVelScale=<0.55 0.55 0.45>
    gravityOverride {z=500}
    physicsBodyDefinitions [{
      collisionType=CustomSphere
      sphereCollisionRadius=2
      boundBone {boneName=dyn_fronthairA_01}
      boxExtents=<30 30 30>
      constraintSetup {angularLimitsMax=<0 0 20>, angularLimitsMin=<0 0 0>, angularTarget=<0 0 0>}
      localJointOffset=<5 0.2 -0.1>
     }]
    relativeSpaceBone {boneName=root}
   }
   n3 = Anim Dynamics - Bone: dyn_fronthairA_02!(ComponentPose=@n2, Alpha=1, GravityOverride:Vector="0,0,-980", ExternalForce="0,0,0") -> (Pose) {
    angularDampingOverride=0.999
    angularSpringConstant=800
    bAngularSpring=true
    bOverrideAngularDamping=true
    bUseGravityOverride=true
    gravityScale=0.3
    simulationSpace=BoneRelative
    alphaBoolBlend {blendOption=Linear}
    boundBone {boneName=dyn_fronthairA_02}
    componentAppliedLinearAccClamp=<2500 2500 1500>
    componentLinearAccScale=<0.5 0.5 0.4>
    componentLinearVelScale=<0.55 0.55 0.45>
    gravityOverride {z=500}
    physicsBodyDefinitions [{
      collisionType=CustomSphere
      sphereCollisionRadius=2
      boundBone {boneName=dyn_fronthairA_02}
      boxExtents=<20 20 20>
      constraintSetup {angularLimitsMax=<0 0 20>, angularLimitsMin=<0 0 0>, angularTarget=<0 0 0>}
      localJointOffset=<5 0.2 -0.1>
     }]
    relativeSpaceBone {boneName=root}
   }
   n4 = Anim Dynamics - Bone: dyn_fronthairB_01!(ComponentPose=@n3, Alpha=1, GravityOverride:Vector="0,0,-980", ExternalForce="0,0,0") -> (Pose) {
    angularDampingOverride=0.999
    angularSpringConstant=800
    bAngularSpring=true
    bOverrideAngularDamping=true
    bUseGravityOverride=true
    gravityScale=0.3
    simulationSpace=BoneRelative
    alphaBoolBlend {blendOption=Linear}
    boundBone {boneName=dyn_fronthairB_01}
    componentAppliedLinearAccClamp=<2500 2500 1500>
    componentLinearAccScale=<0.5 0.5 0.4>
    componentLinearVelScale=<0.55 0.55 0.45>
    gravityOverride {z=500}
    physicsBodyDefinitions [{
      collisionType=CustomSphere
      sphereCollisionRadius=2
      boundBone {boneName=dyn_fronthairB_01}
      boxExtents=<35 35 35>
      constraintSetup {angularLimitsMax=<0 15 20>, angularLimitsMin=<0 -3 -7>, angularTarget=<0 0 0>}
      localJointOffset=<5 0.2 -0.1>
     }]
    relativeSpaceBone {boneName=root}
   }
   n5 = Anim Dynamics - Bone: dyn_fronthairB_02!(ComponentPose=@n4, Alpha=1, GravityOverride:Vector="0,0,-980", ExternalForce="0,0,0") -> (Pose) {
    angularDampingOverride=0.999
    angularSpringConstant=800
    bAngularSpring=true
    bOverrideAngularDamping=true
    bUseGravityOverride=true
    gravityScale=0.3
    simulationSpace=BoneRelative
    alphaBoolBlend {blendOption=Linear}
    boundBone {boneName=dyn_fronthairB_02}
    componentAppliedLinearAccClamp=<2500 2500 1500>
    componentLinearAccScale=<0.5 0.5 0.4>
    componentLinearVelScale=<0.55 0.55 0.45>
    gravityOverride {z=500}
    physicsBodyDefinitions [{
      collisionType=CustomSphere
      sphereCollisionRadius=2
      boundBone {boneName=dyn_fronthairB_02}
      boxExtents=<27 27 27>
      constraintSetup {angularLimitsMax=<0 15 20>, angularLimitsMin=<0 -3 -7>, angularTarget=<0 0 0>}
      localJointOffset=<5 0.2 -0.1>
     }]
    relativeSpaceBone {boneName=root}
   }
   n6 = Anim Dynamics - Bone: dyn_fronthairC_01!(ComponentPose=@n5, Alpha=1, GravityOverride:Vector="0,0,-980", ExternalForce="0,0,0") -> (Pose) {
    angularDampingOverride=0.999
    angularSpringConstant=800
    bAngularSpring=true
    bOverrideAngularDamping=true
    bUseGravityOverride=true
    gravityScale=0.3
    simulationSpace=BoneRelative
    alphaBoolBlend {blendOption=Linear}
    boundBone {boneName=dyn_fronthairC_01}
    componentAppliedLinearAccClamp=<2500 2500 1500>
    componentLinearAccScale=<0.5 0.5 0.4>
    componentLinearVelScale=<0.55 0.55 0.45>
    gravityOverride {z=500}
    physicsBodyDefinitions [{
      collisionType=CustomSphere
      sphereCollisionRadius=2
      boundBone {boneName=dyn_fronthairC_01}
      boxExtents=<35 35 35>
      constraintSetup {angularLimitsMax=<0 15 20>, angularLimitsMin=<0 -3 -7>, angularTarget=<0 0 0>}
      localJointOffset=<5 0.2 -0.1>
     }]
    relativeSpaceBone {boneName=root}
   }
   n7 = Anim Dynamics - Bone: dyn_fronthairC_02!(ComponentPose=@n6, Alpha=1, GravityOverride:Vector="0,0,-980", ExternalForce="0,0,0") -> (Pose) {
    angularDampingOverride=0.999
    angularSpringConstant=800
    bAngularSpring=true
    bOverrideAngularDamping=true
    bUseGravityOverride=true
    gravityScale=0.3
    simulationSpace=BoneRelative
    alphaBoolBlend {blendOption=Linear}
    boundBone {boneName=dyn_fronthairC_02}
    componentAppliedLinearAccClamp=<2500 2500 1500>
    componentLinearAccScale=<0.5 0.5 0.4>
    componentLinearVelScale=<0.55 0.55 0.45>
    gravityOverride {z=500}
    physicsBodyDefinitions [{
      collisionType=CustomSphere
      sphereCollisionRadius=2
      boundBone {boneName=dyn_fronthairC_02}
      boxExtents=<27 27 27>
      constraintSetup {angularLimitsMax=<0 15 20>, angularLimitsMin=<0 -3 -7>, angularTarget=<0 0 0>}
      localJointOffset=<5 0.2 -0.1>
     }]
    relativeSpaceBone {boneName=root}
   }
   n8 = Anim Dynamics - Bone: dyn_fronthairD_01!(ComponentPose=@n7, Alpha=1, GravityOverride:Vector="0,0,-980", ExternalForce="0,0,0") -> (Pose) {
    angularDampingOverride=0.999
    angularSpringConstant=800
    bAngularSpring=true
    bOverrideAngularDamping=true
    bUseGravityOverride=true
    gravityScale=0.3
    simulationSpace=BoneRelative
    alphaBoolBlend {blendOption=Linear}
    boundBone {boneName=dyn_fronthairD_01}
    componentAppliedLinearAccClamp=<2500 2500 1500>
    componentLinearAccScale=<0.5 0.5 0.4>
    componentLinearVelScale=<0.55 0.55 0.45>
    gravityOverride {z=500}
    physicsBodyDefinitions [{
      collisionType=CustomSphere
      sphereCollisionRadius=2
      boundBone {boneName=dyn_fronthairD_01}
      boxExtents=<35 35 35>
      constraintSetup {angularLimitsMax=<0 0 20>, angularLimitsMin=<0 0 0>, angularTarget=<0 0 0>}
      localJointOffset=<5 0.2 -0.1>
     }]
    relativeSpaceBone {boneName=root}
   }
   n9 = Anim Dynamics - Bone: dyn_fronthairD_02!(ComponentPose=@n8, Alpha=1, GravityOverride:Vector="0,0,-980", ExternalForce="0,0,0") -> (Pose) {
    angularDampingOverride=0.999
    angularSpringConstant=800
    bAngularSpring=true
    bOverrideAngularDamping=true
    bUseGravityOverride=true
    gravityScale=0.3
    simulationSpace=BoneRelative
    alphaBoolBlend {blendOption=Linear}
    boundBone {boneName=dyn_fronthairD_02}
    componentAppliedLinearAccClamp=<2500 2500 1500>
    componentLinearAccScale=<0.5 0.5 0.4>
    componentLinearVelScale=<0.55 0.55 0.45>
    gravityOverride {z=500}
    physicsBodyDefinitions [{
      collisionType=CustomSphere
      sphereCollisionRadius=2
      boundBone {boneName=dyn_fronthairD_02}
      boxExtents=<27 27 27>
      constraintSetup {angularLimitsMax=<0 0 20>, angularLimitsMin=<0 0 0>, angularTarget=<0 0 0>}
      localJointOffset=<5 0.2 -0.1>
     }]
    relativeSpaceBone {boneName=root}
   }
   n10 = RigidBody(ComponentPose=@n9) {
    bOverrideWorldGravity=true
    overridePhysicsAsset=&/F_MED_ToadCycle_ArmC_FaceAcc_Physics
    simulationSpace=BaseBoneSpace
    alphaBoolBlend {blendOption=Linear}
    baseBoneRef {boneName=head}
    componentAppliedLinearAccClamp=<2000 2000 2000>
    componentLinearAccScale=<0.3 0.3 0.35>
    componentLinearVelScale=<0.3 0.3 0.35>
    overrideWorldGravity {z=-980}
   }
   Output Pose(Result=Component To Local(ComponentPose=@n10)) {layerGroup=None}
  ]}
 EventGraph {nodes [
   Event BlueprintUpdateAnimation()
   > n1 = Sequence!(execute) -> (then_0, then_1, then_2:Exec)
   n2 = Branch(execute=@n1.then_1, Condition=CharacterItemDefinitionHasAnyMetaTag(Tags=Make Array("[0]"=(TagName="Cosmetics.MetaTags.ToadCycleArm"))))
   > Set On/Off_Arm(On/Off_Arm:bool=true) -> (Output_Get:bool)
   Set On/Off_Arm(execute=@n2.else, On/Off_Arm=false)
   n3 = Branch(execute=@n1.then_2, Condition=CharacterItemDefinitionHasAnyMetaTag(Tags=Make Array("[0]"=(TagName="Cosmetics.MetaTags.ToadCycleSarf"))))
   > Set On/Off_Scarf(On/Off_Scarf:double=1) -> (Output_Get:double)
   Set On/Off_Scarf(execute=@n3.else, On/Off_Scarf=0)
  ]}
]
variables [
 On/Off_Arm {type=boolean}
 On/Off_Scarf {type=number}
]
